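# Sound cue for the third-person looping music of the Hoppin emote, with a movement-reactive stem.
# Beat and moving intro waves play at once, while their looping counterparts start after a short fixed delay.
# Continuous modulators scale the moving stems' volume by the EmoteMoveSpeed parameter, layering them over the unmodulated beat.
# It uses emote attenuation and concurrency, keeps playing when silent, and sends to a music analysis submix.
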SoundCue /BRCosmetics/Sounds/Emotes/Hoppin/SC_Emote_Hoppin_MX_Stem_Loop_3p_Cue
root=Mixer_1
nodes [
 {id=Delay_1, children [Mixer_3], properties {delayMax=0.628, delayMin=0.628}}
 {id=Mixer_1, children [Mixer_2,Delay_1], properties {inputVolume [1,1]}}
 {id=Mixer_2, children [WavePlayer_2, ModulatorContinuous_0], properties {inputVolume [1,1]}}
 {id=Mixer_3, children [WavePlayer_4, ModulatorContinuous_1], properties {inputVolume [1,1]}}
 {
  id=ModulatorContinuous_0
  children [WavePlayer_3]
  properties {pitchModulationParams {minOutput=1, parameterName=EmoteMoveSpeed}, volumeModulationParams {parameterName=EmoteMoveSpeed}}
 }
 {
  id=ModulatorContinuous_1
  children [WavePlayer_5]
  properties {pitchModulationParams {minOutput=1, parameterName=EmoteMoveSpeed}, volumeModulationParams {parameterName=EmoteMoveSpeed}}
 }
 {id=WavePlayer_2, properties {soundWaveAssetPtr=&/Emote_Hoppin_MX_Beat_Intro_01}}
 {id=WavePlayer_3, properties {soundWaveAssetPtr=&/Emote_Hoppin_MX_Moving_Intro_01}}
 {id=WavePlayer_4, properties {bLooping=true, soundWaveAssetPtr=&/Emote_Hoppin_MX_Beat_Loop_01}}
 {id=WavePlayer_5, properties {bLooping=true, soundWaveAssetPtr=&/Emote_Hoppin_MX_Moving_Loop_01}}
]
properties {
 attenuationSettings=&Fort_Emotes_Athena_Att
 maxDistance=1600
 soundClassObject=/Game/Sounds/FortSoundClassesAndModes/Fort_Music_Emote_3P
 virtualizationMode=PlayWhenSilent
 volumeMultiplier=1
 concurrencySet [/Game/Sounds/Concurrency_Rules/Athena_EmoteMusic_Conc]
 soundSubmixSends [{minSendLevel=0, soundSubmix=&Fortnite_AudioAnalysis_Music}]
}
